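# Unlit test-content material in ImpostorBaker that procedurally generates a cloud density field laid out as a pseudo-volume grid of slices.
# The Encode Volume Coords subgraph turns tiled UVs and tile index into 3D coordinates, which Aspect and ZBias reshape into UVW.
# A spherical falloff with power-warped Z, shaped by Bias and Density, is summed with tiled 3D gradient noise to break up the shape.
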Material /ImpostorBaker/TestContent/Cloud/M_GenerateCloud_01
graph {
 properties {bAutomaticallySetUsageInEditor=false, shadingModel=MSM_Unlit, usedShadingModels=Unlit}
 nodes [
  n1 = NamedRerouteUsage()
  n2 = ComponentMask(None=@n1) {b=true}
  n3 = ScalarParameter() {defaultValue=0.5, parameterName=Zpow}
  n4 = Subtract(A=AppendVector(A=ComponentMask(None=@n1) {g=true, r=true}, B=Power(Base=@n2, Exp=@n3))) {constB=0.5}
  n5 = MaterialFunctionCall(Vector 3=@n4, Vector 2=@n4) {materialFunction=/Engine/Functions/Engine_MaterialFunctions02/Utility/VectorLength}
  n6 = Multiply(A=Subtract(A=OneMinus(None=@n5.V3 Length)) {constB=0.5}) {constB=2}
  n7 = ScalarParameter() {defaultValue=0.25, parameterName=Bias}
  Add(A=Multiply(A=Subtract(A=@n6, B=@n7), B=ScalarParameter() {defaultValue=2, parameterName=Density}), B=NamedRerouteUsage())
  n8 = Constant2Vector() {g=16, r=16}
  Composite!(InputPin=@n8) {subgraphName=Encode Volume Coords}
  MaterialFunctionCall!(X range Y range=Constant4Vector() {constant=<rgba 0 1 0 1>}.RGBA, Gradient Height, ShowGrid?, Thickness, Color, "f(x)") {materialFunction=/Engine/Functions/Engine_MaterialFunctions01/Debug/PlotFunctionOnGraph}
  MaterialExpressionReroute_6 = Reroute(None=@n8)
  n9 = Multiply(A=TextureCoordinate(), B=ComponentMask(None=@MaterialExpressionReroute_6.InputPin) {g=true, r=true})
  n10 = MaterialFunctionCall(Float2=@MaterialExpressionReroute_6.InputPin) {materialFunction=&BreakOutFloat2Components}
  n11 = MaterialFunctionCall!(2D Index=Floor(None=@n9), Columns X=@n10.R) {materialFunction=/Engine/Functions/Engine_MaterialFunctions01/Coordinates/2Dto1DIndex}
  n12 = AppendVector(A=Frac(None=@n9), B=Divide(A=Add(A=@n11.1D Index) {constB=0.5}, B=Multiply(A=@n10.R, B=@n10.R)))
  MaterialExpressionReroute_5 = Reroute(None=@n12)
  n13 = VectorParameter() {parameterName=Aspect, defaultValue=<rgba 1 1 1.5 0>}
  NamedRerouteDeclaration(None=Subtract(A=Multiply(A=@MaterialExpressionReroute_5.Output, B=@n13.RGB), B=VectorParameter() {parameterName=ZBias, defaultValue=<rgba 0 0 0.35 0>}.RGB)) {name=UVW, nodeColor=<rgba 1 0 0.75 1>}
  n14 = Multiply(
   A=Divide(A=NamedRerouteUsage(), B=VectorParameter() {parameterName=Aspect, defaultValue=<rgba 1 1 1.5 0>}.RGB)
   B=ScalarParameter() {defaultValue=8, parameterName=NoiseTiling}
  )
  NamedRerouteDeclaration(None=Multiply(
    A=Noise!(World Position=@n14, FilterWidth) {noiseFunction=NOISEFUNCTION_GradientTex3D}
    B=ScalarParameter() {defaultValue=0.3, parameterName=Noise Amount}
   )) {name=ProcNoise, nodeColor=<rgba 1 0 0.6797 1>}
  PinBase() {reroutePins [InputPin {expression=&self:MaterialExpressionReroute_6}]}
  PinBase!(Output=@n12) {reroutePins [Output {expression=&self:MaterialExpressionReroute_5}]}
 ]
}
