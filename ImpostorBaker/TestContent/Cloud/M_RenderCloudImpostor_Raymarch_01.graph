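# Unlit alpha-composite test material that raymarches a pseudo-volume cloud texture for impostor baking.
# A custom cube-intersection node supplies ray entry and step count to a jittered density march, which outputs opacity, mean depth and self-shadowing for four view-space light directions.
# MPC_ImposterCapture parameters blend between custom-lighting, opacity and depth capture outputs.
# A separate term weights values by a render light vector's alignment with five basis directions.
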Material /ImpostorBaker/TestContent/Cloud/M_RenderCloudImpostor_Raymarch_01
graph {
 properties {bDisableDepthTest=true, blendMode=BLEND_AlphaComposite, pixelDepthOffsetMode=PDOM_Legacy, shadingModel=MSM_Unlit, usedShadingModels=Unlit}
 nodes [
  n1 = Normalize(VectorInput=Transform(None=VectorParameter() {parameterName=LightVector_3, defaultValue=<rgba 0 1 0 1>}.RGB) {transformSourceType=TRANSFORMSOURCE_View, transformType=TRANSFORM_Local})
  n2 = Multiply(A=ScreenPosition().ViewportUV) {constB=3}
  n3 = Transform(None=Normalize(VectorInput=VectorParameter() {parameterName=Render Light Vector, defaultValue=<rgba 1 0 0 0>}.RGB)) {transformSourceType=TRANSFORMSOURCE_World, transformType=TRANSFORM_Local}
  n4 = Normalize(VectorInput=Transform(None=VectorParameter() {parameterName=LightVector_1, defaultValue=<rgba 1 0 0 1>}.RGB) {transformSourceType=TRANSFORMSOURCE_View, transformType=TRANSFORM_Local})
  n5 = Normalize(VectorInput=Transform(None=VectorParameter() {parameterName=LightVector_2, defaultValue=<rgba -1 0 0 1>}.RGB) {transformSourceType=TRANSFORMSOURCE_View, transformType=TRANSFORM_Local})
  n6 = VectorParameter() {parameterName=LightVector_4, defaultValue=<rgba 0 0 1 1>}
  n7 = DotProduct(A=@n3, B=Normalize(VectorInput=Transform(None=@n6.RGB) {transformSourceType=TRANSFORMSOURCE_View, transformType=TRANSFORM_Local}))
  n8 = ScalarParameter() {defaultValue=1, parameterName=Zpower}
  n9 = Max(A=MaterialFunctionCall(X=DotProduct(A=@n3, B=@n4), Y=DotProduct(A=@n3, B=@n5), Z=DotProduct(A=@n3, B=@n1), A=Power(Base=@n7, Exp=@n8)) {materialFunction=&MakeFloat4}.Result) {constB=0}
  n10 = NamedRerouteUsage()
  n11 = Transform(None=VectorParameter() {parameterName=LightVector_5, defaultValue=<rgba 0 0 -1 1>}.RGB) {transformSourceType=TRANSFORMSOURCE_View, transformType=TRANSFORM_Local}
  n12 = Max(A=DotProduct(A=@n3, B=Normalize(VectorInput=@n11))) {constB=0}
  n13 = NamedRerouteUsage()
  Add(
   A=MaterialFunctionCall!(Vector3=@n1, MaximumNumberOfDigits, UVs=Frac(None=@n2), DebugTextLocation RG_UpperRight BA_LowerLeft, Component Spacing) {materialFunction=/Engine/Functions/Engine_MaterialFunctions02/Utility/DebugFloat3Values}.ColorCodedOutput
   B=Add(A=DotProduct(A=@n9, B=@n10), B=Multiply(A=@n12, B=@n13))
  )
  Constant() {r=1}
  Constant() {r=1}
  n14 = Constant() {r=1}
  If(A=NamedRerouteUsage(), "A > B"=@n14, "A == B"=Constant(), "A < B"=@n14) {constB=0.5}
  n15 = LinearInterpolate(
   A=NamedRerouteUsage()
   B=NamedRerouteUsage()
   Alpha=CollectionParameter() {collection=/ImpostorBaker/Materials/MPC/MPC_ImposterCapture, parameterName=CustomLighting}
  ) {constA=1}
  n16 = NamedRerouteUsage()
  n17 = CollectionParameter() {collection=&MPC_ImposterCapture, parameterName=Opacity}
  LinearInterpolate(
   A=NamedRerouteUsage()
   B=LinearInterpolate(
    A=LinearInterpolate(A=@n15, B=@n16, Alpha=@n17)
    B=NamedRerouteUsage()
    Alpha=CollectionParameter() {collection=&MPC_ImposterCapture, parameterName=Depth}
   )
   Alpha=CollectionParameter() {collection=&MPC_ImposterCapture, parameterName=EnableCapture}
  )
  n18 = NamedRerouteUsage()
  n19 = CollectionParameter() {collection=&MPC_ImposterCapture, parameterName=CustomLighting}
  n20 = NamedRerouteUsage()
  n21 = ScalarParameter() {defaultValue=0.03, parameterName=Depth Opacity Cut}
  n22 = Saturate(None=Ceil(None=Subtract(A=@n20, B=@n21)))
  LinearInterpolate(
   A=NamedRerouteUsage()
   B=LinearInterpolate(
    A=LinearInterpolate(B=@n18, Alpha=@n19) {constA=1}
    B=LinearInterpolate(Alpha=@n22)
    Alpha=CollectionParameter() {collection=&MPC_ImposterCapture, parameterName=Depth}
   )
   Alpha=CollectionParameter() {collection=&MPC_ImposterCapture, parameterName=EnableCapture}
  )
  MaterialExpressionNamedRerouteUsage_26 = NamedRerouteUsage()
  MaterialExpressionNamedRerouteUsage_25 = NamedRerouteUsage()
  MaterialExpressionNamedRerouteUsage_24 = NamedRerouteUsage()
  MaterialExpressionNamedRerouteUsage_21 = NamedRerouteUsage()
  MaterialExpressionNamedRerouteUsage_22 = NamedRerouteUsage()
  MaterialExpressionNamedRerouteUsage_23 = NamedRerouteUsage()
  MaterialExpressionNamedRerouteUsage_27 = NamedRerouteUsage()
  MaterialExpressionNamedRerouteUsage_28 = NamedRerouteUsage()
  MaterialExpressionNamedRerouteUsage_29 = NamedRerouteUsage()
  MaterialExpressionNamedRerouteUsage_30 = NamedRerouteUsage()
  MaterialExpressionNamedRerouteUsage_31 = NamedRerouteUsage()
  MaterialExpressionNamedRerouteUsage_32 = NamedRerouteUsage()
  MaterialExpressionNamedRerouteUsage_17 = NamedRerouteUsage()
  MaterialExpressionNamedRerouteUsage_18 = NamedRerouteUsage()
  MaterialExpressionNamedRerouteUsage_19 = NamedRerouteUsage()
  MaterialExpressionNamedRerouteUsage_20 = NamedRerouteUsage()
  n23 = Custom!(
   Tex=@MaterialExpressionNamedRerouteUsage_26
   XYFrames=@MaterialExpressionNamedRerouteUsage_25
   StepSize=@MaterialExpressionNamedRerouteUsage_24
   MaxSteps=@MaterialExpressionNamedRerouteUsage_21
   FinalStep=@MaterialExpressionNamedRerouteUsage_22
   CurPos=@MaterialExpressionNamedRerouteUsage_23
   Density=@MaterialExpressionNamedRerouteUsage_27
   ShadowSteps=@MaterialExpressionNamedRerouteUsage_28
   ShadowDensity=@MaterialExpressionNamedRerouteUsage_29
   ShadowThreshold=@MaterialExpressionNamedRerouteUsage_30
   Jitter=@MaterialExpressionNamedRerouteUsage_31
   CaptureZ=@MaterialExpressionNamedRerouteUsage_32
   LightVector=@MaterialExpressionNamedRerouteUsage_17
   LightVector2=@MaterialExpressionNamedRerouteUsage_18
   LightVector3=@MaterialExpressionNamedRerouteUsage_19
   LightVector4=@MaterialExpressionNamedRerouteUsage_20
  ) {
   code="float numFrames = XYFrames.x * XYFrames.y;\r\nfloat accumdist = 0;\r\nfloat curdensity = 0;\r\nfloat transmittance = 1;\r\nfloat3 localcamvec = normalize( mul(Parameters.CameraVector, (float3x3)WSDemote(GetWorldToLocal(Parameters))) ) * StepSize;\r\n\r\nfloat shadowstepsize = 1 / ShadowSteps;\r\nLightVector *= shadowstepsize*0.5;\r\nLightVector2 *= shadowstepsize*0.5;\r\nLightVector3 *= shadowstepsize*0.5;\r\nLightVector4 *= shadowstepsize*0.5;\r\n\r\nShadowDensity *= shadowstepsize;\r\n\r\nDensity *= StepSize;\r\n\r\nfloat lightenergy = 0;\r\nfloat lightenergy2 = 0;\r\nfloat lightenergy3 = 0;\r\nfloat lightenergy4 = 0;\r\n\r\nfloat shadowthresh = -log(ShadowThreshold) / ShadowDensity;\r\n\r\nint3 randpos = int3(Parameters.SvPosition.xy, View.StateFrameIndexMod8);\r\nfloat rand =float(Rand3DPCG16(randpos).x) / 0xffff;\r\nCurPos +=  localcamvec * rand.x * Jitter;\r\n\r\n//input v3 CaptureZ\r\n//input float CaptureRadius\r\n\r\nfloat MeanDepth = 0.5;\r\n\r\nfor (int i = 0; i < MaxSteps; i++)\r\n{\t\r\n\r\n\t\r\n\tfloat cursample = PseudoVolumeTexture(Tex, TexSampler, CurPos, XYFrames, numFrames).r;\r\n\t\r\n\t//Sample Light Absorption and Scattering\r\n\tif( (cursample.r) > 0.001)\r\n\t{\r\n\t\tfloat3 lpos = CurPos;\r\n\t\tfloat shadowdist = 0;\r\n\t\t\r\n\t\tcurdensity = 1 - exp(-cursample.r * Density);\r\n\r\n\t\tfor (int s = 0; s < ShadowSteps; s++)\r\n\t\t{\r\n\t\t\tlpos += LightVector;\r\n\r\n\t\t\r\n\r\n\t\t\tfloat lsample = PseudoVolumeTexture(Tex, TexSampler, saturate(lpos), XYFrames, numFrames).r;\r\n\t\t\t\r\n\t\t\tfloat3 shadowboxtest = floor( 0.5 + ( abs( 0.5 - lpos ) ) );\r\n\t\t\tfloat exitshadowbox = shadowboxtest .x + shadowboxtest .y + shadowboxtest .z;\r\n\r\n       \t\t\tif(shadowdist > shadowthresh || exitshadowbox >= 1) break;\r\n\r\n\t\t\tshadowdist += lsample;\r\n\t\t}\r\n\r\n\t\tlightenergy += exp(-shadowdist * ShadowDensity) * curdensity * transmittance;\r\n\t\t\r\n\t\t//sample second light\r\n\t\t\tlpos = CurPos;\r\n\t\t\tshadowdist = 0;\r\n\t\t\tfor (int s2 = 0; s2 < ShadowSteps; s2++)\r\n\t\t\t{\r\n\t\t\t\tlpos += LightVector2;\r\n\t\r\n\t\t\t\r\n\t\r\n\t\t\t\tfloat lsample = PseudoVolumeTexture(Tex, TexSampler, saturate(lpos), XYFrames, numFrames).r;\r\n\t\t\t\t\r\n\t\t\t\tfloat3 shadowboxtest = floor( 0.5 + ( abs( 0.5 - lpos ) ) );\r\n\t\t\t\tfloat exitshadowbox = shadowboxtest .x + shadowboxtest .y + shadowboxtest .z;\r\n\t\r\n\t\t\t\t\tif(shadowdist > shadowthresh || exitshadowbox >= 1) break;\r\n\t\r\n\t\t\t\tshadowdist += lsample;\r\n\t\t\t}\r\n\t\r\n\t\t\tlightenergy2 += exp(-shadowdist * ShadowDensity) * curdensity * transmittance;\t\r\n\r\n\t\t//sample third light\r\n\t\t\tlpos = CurPos;\r\n\t\t\tshadowdist = 0;\r\n\t\t\tfor (int s3 = 0; s3 < ShadowSteps; s3++)\r\n\t\t\t{\r\n\t\t\t\tlpos += LightVector3;\r\n\t\r\n\t\t\t\r\n\t\r\n\t\t\t\tfloat lsample = PseudoVolumeTexture(Tex, TexSampler, saturate(lpos), XYFrames, numFrames).r;\r\n\t\t\t\t\r\n\t\t\t\tfloat3 shadowboxtest = floor( 0.5 + ( abs( 0.5 - lpos ) ) );\r\n\t\t\t\tfloat exitshadowbox = shadowboxtest .x + shadowboxtest .y + shadowboxtest .z;\r\n\t\r\n\t\t\t\t\tif(shadowdist > shadowthresh || exitshadowbox >= 1) break;\r\n\t\r\n\t\t\t\tshadowdist += lsample;\r\n\t\t\t}\r\n\t\r\n\t\t\tlightenergy3 += exp(-shadowdist * ShadowDensity) * curdensity * transmittance;\t\r\n\r\n\t\t//sample fourth light\r\n\t\t\tlpos = CurPos;\r\n\t\t\tshadowdist = 0;\r\n\t\t\tfor (int s4 = 0; s4 < ShadowSteps; s4++)\r\n\t\t\t{\r\n\t\t\t\tlpos += LightVector4;\r\n\t\r\n\t\t\t\r\n\t\r\n\t\t\t\tfloat lsample = PseudoVolumeTexture(Tex, TexSampler, saturate(lpos), XYFrames, numFrames).r;\r\n\t\t\t\t\r\n\t\t\t\tfloat3 shadowboxtest = floor( 0.5 + ( abs( 0.5 - lpos ) ) );\r\n\t\t\t\tfloat exitshadowbox = shadowboxtest .x + shadowboxtest .y + shadowboxtest .z;\r\n\t\r\n\t\t\t\t\tif(shadowdist > shadowthresh || exitshadowbox >= 1) break;\r\n\t\r\n\t\t\t\tshadowdist += lsample;\r\n\t\t\t}\r\n\t\r\n\t\t\tlightenergy4 += exp(-shadowdist * ShadowDensity) * curdensity * transmittance;\t\r\n\r\n\r\n\t\t\t\r\nif(curdensity > 0.05)\r\n{\r\n\t\ttransmittance *= 1- (curdensity);\r\n\t\t\r\n\t\tfloat LocalDepth = dot(CurPos - 0.5, CaptureZ);\r\n\t\t//LocalDepth /= CaptureRadius;\r\n\t\tLocalDepth = 1 - saturate(LocalDepth + 0.5);\r\n\t\t\r\n\t\tMeanDepth = lerp(MeanDepth, LocalDepth, saturate(3.0*(transmittance)*curdensity));\r\n}\r\n\t\t\r\n\t\t##if 0\r\n\r\n\t\t//Sky Lighting\r\n\t\tshadowdist = 0;\r\n\t\t\r\n\t\tlpos = CurPos + float3(0,0,0.025);\r\n\t\tfloat lsample = PseudoVolumeTexture(Tex, TexSampler, saturate(lpos), XYFrames, numFrames).r;\r\n\t\tshadowdist += lsample;\r\n\t\tlpos = CurPos + float3(0,0,0.05);\r\n\t\tlsample = PseudoVolumeTexture(Tex, TexSampler, saturate(lpos), XYFrames, numFrames).r;\r\n\t\tshadowdist += lsample;\r\n\t\tlpos = CurPos + float3(0,0,0.15);\r\n\t\tlsample = PseudoVolumeTexture(Tex, TexSampler, saturate(lpos), XYFrames, numFrames).r;\r\n\t\tshadowdist += lsample;\r\n\r\n\t\tlightenergy    += exp(-shadowdist  * AmbientDensity) * curdensity * SkyColor * transmittance;\r\n\t\t##endif\r\n\r\n\t}\r\n\t\t\r\n\r\n\tCurPos -= localcamvec;\r\n}\r\n\r\nOutMeanDepth = MeanDepth;\r\nOutShadow1 = lightenergy;\r\nOutShadow2 = lightenergy2;\r\nOutShadow3 = lightenergy3;\r\nOutShadow4 = lightenergy4;\r\n\r\nreturn float4( lightenergy.xxx, transmittance);\r\n"
   desc="\n*** Original source before expression upgrade ***\nfloat numFrames = XYFrames.x * XYFrames.y;\r\nfloat accumdist = 0;\r\nfloat curdensity = 0;\r\nfloat transmittance = 1;\r\nfloat3 localcamvec = normalize( mul(Parameters.CameraVector, (float3x3)GetPrimitiveData(Parameters).WorldToLocal) ) * StepSize;\r\n\r\nfloat3 invlightdir = 1 / LightVector;\r\n\r\nfloat shadowstepsize = 1 / ShadowSteps;\r\nLightVector *= shadowstepsize*0.5;\r\nShadowDensity *= shadowstepsize;\r\n\r\nDensity *= StepSize;\r\nfloat3 lightenergy = 0;\r\nfloat shadowthresh = -log(ShadowThreshold) / ShadowDensity;\r\n\r\nint3 randpos = int3(Parameters.SvPosition.xy, View.StateFrameIndexMod8);\r\nfloat rand =float(Rand3DPCG16(randpos).x) / 0xffff;\r\nCurPos +=  localcamvec * rand.x * Jitter;\r\n\r\n\r\nfor (int i = 0; i < MaxSteps; i++)\r\n{\t\r\n\r\n\t\r\n\tfloat cursample = PseudoVolumeTexture(Tex, TexSampler, CurPos, XYFrames, numFrames).r;\r\n\t\r\n\t//Sample Light Absorption and Scattering\r\n\tif( (cursample.r) > 0.001)\r\n\t{\r\n\t\tfloat3 lpos = CurPos;\r\n\t\tfloat shadowdist = 0;\r\n\r\n\t\tfor (int s = 0; s < ShadowSteps; s++)\r\n\t\t{\r\n\t\t\tlpos += LightVector;\r\n\r\n\t\t\r\n\r\n\t\t\tfloat lsample = PseudoVolumeTexture(Tex, TexSampler, saturate(lpos), XYFrames, numFrames).r;\r\n\t\t\t\r\n\t\t\tfloat3 shadowboxtest = floor( 0.5 + ( abs( 0.5 - lpos ) ) );\r\n\t\t\tfloat exitshadowbox = shadowboxtest .x + shadowboxtest .y + shadowboxtest .z;\r\n\r\n       \t\t\tif(shadowdist > shadowthresh || exitshadowbox >= 1) break;\r\n\r\n\t\t\tshadowdist += lsample;\r\n\t\t}\r\n\t\r\n\t\tcurdensity = 1 - exp(-cursample.r * Density);\r\n\r\n\r\n\t\t//curdensity = saturate(cursample * Density);\r\n\t\t//float     shadowterm = exp(-shadowdist * ShadowDensity);\r\n\t\t//float3 absorbedlight = exp(-shadowdist * ShadowDensity) * curdensity;\r\n\r\n\t\tlightenergy += exp(-shadowdist * ShadowDensity) * curdensity * transmittance * LightColor;\r\n\r\n\t\ttransmittance *= 1- (curdensity);\r\n\t\t\r\n\t\t##if 1\r\n\r\n\t\t//Sky Lighting\r\n\t\tshadowdist = 0;\r\n\t\t\r\n\t\tlpos = CurPos + float3(0,0,0.025);\r\n\t\tfloat lsample = PseudoVolumeTexture(Tex, TexSampler, saturate(lpos), XYFrames, numFrames).r;\r\n\t\tshadowdist += lsample;\r\n\t\tlpos = CurPos + float3(0,0,0.05);\r\n\t\tlsample = PseudoVolumeTexture(Tex, TexSampler, saturate(lpos), XYFrames, numFrames).r;\r\n\t\tshadowdist += lsample;\r\n\t\tlpos = CurPos + float3(0,0,0.15);\r\n\t\tlsample = PseudoVolumeTexture(Tex, TexSampler, saturate(lpos), XYFrames, numFrames).r;\r\n\t\tshadowdist += lsample;\r\n\r\n\t\tlightenergy    += exp(-shadowdist  * AmbientDensity) * curdensity * SkyColor * transmittance;\r\n\t\t##endif\r\n\r\n\t}\r\n\t\t\r\n\r\n\tCurPos -= localcamvec;\r\n}\r\n\r\n\r\n\r\nreturn float4( lightenergy, transmittance);\r\n"
   description=Density RayMarch
   outputType=CMOT_Float4
   additionalOutputs [
    {outputName=OutMeanDepth, outputType=CMOT_Float1}
    {outputName=OutShadow1, outputType=CMOT_Float1}
    {outputName=OutShadow2, outputType=CMOT_Float1}
    {outputName=OutShadow3, outputType=CMOT_Float1}
    {outputName=OutShadow4, outputType=CMOT_Float1}
   ]
   inputs [
    {inputName=Tex, input {expression=&self:MaterialExpressionNamedRerouteUsage_26}}
    {
     inputName=XYFrames
     input {expression=&self:MaterialExpressionNamedRerouteUsage_25}
    }
    {
     inputName=StepSize
     input {expression=&self:MaterialExpressionNamedRerouteUsage_24}
    }
    {
     inputName=MaxSteps
     input {expression=&self:MaterialExpressionNamedRerouteUsage_21}
    }
    {
     inputName=FinalStep
     input {expression=&self:MaterialExpressionNamedRerouteUsage_22}
    }
    {
     inputName=CurPos
     input {expression=&self:MaterialExpressionNamedRerouteUsage_23}
    }
    {
     inputName=Density
     input {expression=&self:MaterialExpressionNamedRerouteUsage_27}
    }
    {
     inputName=ShadowSteps
     input {expression=&self:MaterialExpressionNamedRerouteUsage_28}
    }
    {
     inputName=ShadowDensity
     input {expression=&self:MaterialExpressionNamedRerouteUsage_29}
    }
    {
     inputName=ShadowThreshold
     input {expression=&self:MaterialExpressionNamedRerouteUsage_30}
    }
    {
     inputName=Jitter
     input {expression=&self:MaterialExpressionNamedRerouteUsage_31}
    }
    {
     inputName=CaptureZ
     input {expression=&self:MaterialExpressionNamedRerouteUsage_32}
    }
    {
     inputName=LightVector
     input {expression=&self:MaterialExpressionNamedRerouteUsage_17}
    }
    {
     inputName=LightVector2
     input {expression=&self:MaterialExpressionNamedRerouteUsage_18}
    }
    {
     inputName=LightVector3
     input {expression=&self:MaterialExpressionNamedRerouteUsage_19}
    }
    {
     inputName=LightVector4
     input {expression=&self:MaterialExpressionNamedRerouteUsage_20}
    }
   ]
  }
  NamedRerouteDeclaration(None=OneMinus(None=ComponentMask(None=@n23.return) {a=true})) {name=Opacity, nodeColor=<rgba 0.0078 1 0 1>}
  NamedRerouteDeclaration(None=@n23.OutMeanDepth) {name=Mean Depth, nodeColor=<rgba 1 0 0.7734 1>}
  n24 = MaterialFunctionCall(
   X=NamedRerouteDeclaration(None=@n23.OutShadow1) {name=Shadow1, nodeColor=<rgba 1 0 0.3281 1>}
   Y=NamedRerouteDeclaration(None=@n23.OutShadow2) {name=Shadow2, nodeColor=<rgba 1 0 0.3281 1>}
   Z=NamedRerouteDeclaration(None=@n23.OutShadow3) {name=Shadow3, nodeColor=<rgba 1 0 0.3281 1>}
  ) {materialFunction=&MakeFloat3}
  n25 = NamedRerouteDeclaration(None=@n23.OutShadow4) {name=Shadow4, nodeColor=<rgba 1 0 0.3281 1>}
  NamedRerouteDeclaration(None=AppendVector(
    A=NamedRerouteDeclaration(None=@n24.Result) {name=Custom Lighting RGB, nodeColor=<rgba 0 1 0.8828 1>}
    B=NamedRerouteDeclaration(None=@n25) {name=Custom Lighting A, nodeColor=<rgba 1 0.5625 0 1>}
   )) {name=Custom Lighting RGBA, nodeColor=<rgba 1 0.0938 0 1>}
  n26 = VectorParameter() {parameterName=LightVector_1, defaultValue=<rgba 1 0 0 1>}
  NamedRerouteDeclaration(None=Normalize(VectorInput=Transform(None=@n26.RGB) {transformSourceType=TRANSFORMSOURCE_View, transformType=TRANSFORM_Local})) {name=Light Vector 1, nodeColor=<rgba 0.7344 0 1 1>}
  n27 = VectorParameter() {parameterName=LightVector_2, defaultValue=<rgba -1 0 0 1>}
  NamedRerouteDeclaration(None=Normalize(VectorInput=Transform(None=@n27.RGB) {transformSourceType=TRANSFORMSOURCE_View, transformType=TRANSFORM_Local})) {name=Light Vector 2, nodeColor=<rgba 0 1 0.4141 1>}
  n28 = VectorParameter() {parameterName=LightVector_3, defaultValue=<rgba 0 1 0 1>}
  NamedRerouteDeclaration(None=Normalize(VectorInput=Transform(None=@n28.RGB) {transformSourceType=TRANSFORMSOURCE_View, transformType=TRANSFORM_Local})) {name=Light Vector 3, nodeColor=<rgba 1 0.0938 0 1>}
  n29 = VectorParameter() {parameterName=LightVector_4, defaultValue=<rgba 0 0 1 1>}
  NamedRerouteDeclaration(None=Normalize(VectorInput=Transform(None=@n29.RGB) {transformSourceType=TRANSFORMSOURCE_View, transformType=TRANSFORM_Local})) {name=Light Vector 4, nodeColor=<rgba 0 0.2266 1 1>}
  MaterialExpressionScalarParameter_3 = ScalarParameter() {parameterName=Plane Alignment}
  MaterialExpressionScalarParameter_1 = ScalarParameter() {defaultValue=128, parameterName=Steps}
  MaterialExpressionSceneDepth_1 = SceneDepth()
  n30 = Custom!(
   PlaneAlignment=@MaterialExpressionScalarParameter_3
   MaxSteps=@MaterialExpressionScalarParameter_1
   scenedepth=@MaterialExpressionSceneDepth_1.R
  ) {
   code="float localscenedepth = scenedepth;\r\n\r\nfloat3 camerafwd = mul(float3(0.00000000,0.00000000,1.00000000),(float3x3)ResolvedView.ViewToTranslatedWorld);\r\n\r\nfloat3 depthpos = ((Parameters.CameraVector * localscenedepth) / abs( dot( camerafwd, Parameters.CameraVector ) ) );\r\n\r\ndepthpos = mul(depthpos,  (float3x3)WSDemote(GetWorldToLocal(Parameters))).xyz;\r\n\r\ndepthpos /= 256;\r\n\r\nlocalscenedepth = length(depthpos);\r\n\r\n//0-1\r\n//localscenedepth /= (GetPrimitiveData(Parameters).LocalObjectBoundsMax.x * 2 * scale);\r\n//localscenedepth /= abs( dot( camerafwd, Parameters.CameraVector ) );\r\n\r\n//bring vectors into local space to support object transforms\r\nfloat3 localcampos = mul(float4( DFDemote(ResolvedView.WorldCameraOrigin),1.00000000), (WSDemote(GetWorldToLocal(Parameters)))).xyz;\r\nfloat3 localcamvec = -normalize( mul(Parameters.CameraVector, (float3x3)WSDemote(GetWorldToLocal(Parameters))) );\r\n\r\n//make camera position 0-1\r\nlocalcampos = (localcampos / (GetPrimitiveData(Parameters).LocalObjectBoundsMax.x * 2)) + 0.5;\r\n\r\nfloat3 invraydir = 1 / localcamvec;\r\n\r\nfloat3 firstintersections = (0 - localcampos) * invraydir;\r\nfloat3 secondintersections = (1 - localcampos) * invraydir;\r\nfloat3 closest = min(firstintersections, secondintersections);\r\nfloat3 furthest = max(firstintersections, secondintersections);\r\n\r\nfloat t0 = max(closest.x, max(closest.y, closest.z));\r\nfloat t1 = min(furthest.x, min(furthest.y, furthest.z));\r\n\r\nfloat planeoffset = 1-frac( ( t0 - length(localcampos-0.5) ) * MaxSteps );\r\n\r\nt0 += (planeoffset / MaxSteps) * PlaneAlignment;\r\nt1 = min(t1, localscenedepth);\r\nt0 = max(0, t0);\r\n\r\nfloat boxthickness = max(0, t1 - t0);\r\n\r\nfloat3 entrypos = localcampos + (max(0,t0) * localcamvec);\r\n\r\n\r\nreturn float4( entrypos, boxthickness );"
   desc="\n*** Original source before expression upgrade ***\nfloat localscenedepth = scenedepth;\r\n\r\nfloat3 camerafwd = mul(float3(0.00000000,0.00000000,1.00000000),(float3x3)ResolvedView.ViewToTranslatedWorld);\r\n\r\nfloat3 depthpos = ((Parameters.CameraVector * localscenedepth) / abs( dot( camerafwd, Parameters.CameraVector ) ) );\r\n\r\ndepthpos = mul(depthpos,  (float3x3)GetPrimitiveData(Parameters).WorldToLocal).xyz;\r\n\r\ndepthpos /= 256;\r\n\r\nlocalscenedepth = length(depthpos);\r\n\r\n//0-1\r\n//localscenedepth /= (GetPrimitiveData(Parameters).LocalObjectBoundsMax.x * 2 * scale);\r\n//localscenedepth /= abs( dot( camerafwd, Parameters.CameraVector ) );\r\n\r\n//bring vectors into local space to support object transforms\r\nfloat3 localcampos = mul(float4( ResolvedView.WorldCameraOrigin,1.00000000), (GetPrimitiveData(Parameters).WorldToLocal)).xyz;\r\nfloat3 localcamvec = -normalize( mul(Parameters.CameraVector, (float3x3)GetPrimitiveData(Parameters).WorldToLocal) );\r\n\r\n//make camera position 0-1\r\nlocalcampos = (localcampos / (GetPrimitiveData(Parameters).LocalObjectBoundsMax.x * 2)) + 0.5;\r\n\r\nfloat3 invraydir = 1 / localcamvec;\r\n\r\nfloat3 firstintersections = (0 - localcampos) * invraydir;\r\nfloat3 secondintersections = (1 - localcampos) * invraydir;\r\nfloat3 closest = min(firstintersections, secondintersections);\r\nfloat3 furthest = max(firstintersections, secondintersections);\r\n\r\nfloat t0 = max(closest.x, max(closest.y, closest.z));\r\nfloat t1 = min(furthest.x, min(furthest.y, furthest.z));\r\n\r\nfloat planeoffset = 1-frac( ( t0 - length(localcampos-0.5) ) * MaxSteps );\r\n\r\nt0 += (planeoffset / MaxSteps) * PlaneAlignment;\r\nt1 = min(t1, localscenedepth);\r\nt0 = max(0, t0);\r\n\r\nfloat boxthickness = max(0, t1 - t0);\r\n\r\nfloat3 entrypos = localcampos + (max(0,t0) * localcamvec);\r\n\r\n\r\nreturn float4( entrypos, boxthickness );"
   description=Ray March Cube Setup
   outputType=CMOT_Float4
   inputs [
    {
     inputName=PlaneAlignment
     input {expression=&self:MaterialExpressionScalarParameter_3}
    }
    {
     inputName=MaxSteps
     input {expression=&self:MaterialExpressionScalarParameter_1}
    }
    {
     inputName=scenedepth
     input {expression=&self:MaterialExpressionSceneDepth_1, mask=1, maskR=1}
    }
   ]
  }
  n31 = Multiply(A=ComponentMask(None=@n30) {a=true}, B=ScalarParameter() {defaultValue=128, parameterName=Steps})
  NamedRerouteDeclaration(None=Clamp(None=Floor(None=@n31)) {desc=MaxSteps, maxDefault=512}) {name=NumSteps, nodeColor=<rgba 0.5469 1 0 1>}
  NamedRerouteDeclaration(None=Frac(None=@n31) {desc=FinalStepSize}) {name=FinalStep, nodeColor=<rgba 1 0 0.8672 1>}
  NamedRerouteDeclaration(None=Multiply(A=ComponentMask(None=@n30) {b=true, g=true, r=true}) {desc=Ray Entry}) {name=RayOrigin, nodeColor=<rgba 0 1 0.8125 1>}
  NamedRerouteDeclaration() {name=Opacity_1, nodeColor=<rgba 0.0078 1 0 1>}
  NamedRerouteDeclaration(None=Divide(B=ScalarParameter() {defaultValue=128, parameterName=Steps})) {name=StepSize, nodeColor=<rgba 1 0.4922 0 1>}
  NamedRerouteDeclaration(None=ComponentMask(None=VectorParameter() {parameterName=XYFrames, defaultValue=<rgba 4 4 0 0>}.RGB) {g=true, r=true}) {name=XY Frames, nodeColor=<rgba 0.1719 0 1 1>}
  NamedRerouteDeclaration(None=TextureObjectParameter() {parameterName=PseudoVolume, texture=/Engine/EditorResources/T_EditorHelp}) {name=PseudoVolumeTextureObject, nodeColor=<rgba 0.1484 1 0 1>}
  NamedRerouteDeclaration(None=ScalarParameter() {defaultValue=16, parameterName=Density}) {name=Density, nodeColor=<rgba 1 0 0.4687 1>}
  NamedRerouteDeclaration(None=ScalarParameter() {defaultValue=32, parameterName=ShadowSteps}) {name=ShadowSteps, nodeColor=<rgba 0 0.7891 1 1>}
  NamedRerouteDeclaration(None=ScalarParameter() {defaultValue=16, parameterName=ShadowDensity}) {name=ShadowDensity, nodeColor=<rgba 1 0.8906 0 1>}
  NamedRerouteDeclaration(None=ScalarParameter() {defaultValue=0.001, parameterName=ShadowThreshold}) {name=ShadowThreshold, nodeColor=<rgba 0.5703 0 1 1>}
  NamedRerouteDeclaration(None=ScalarParameter() {defaultValue=1, parameterName=Jitter}) {name=Jitter, nodeColor=<rgba 0 1 0.25 1>}
  NamedRerouteDeclaration(None=ComponentMask(None=CollectionParameter() {collection=&MPC_ImposterCapture, parameterName=Z}) {b=true, g=true, r=true}) {name=Z Vector, nodeColor=<rgba 1 0 0.0703 1>}
  NamedRerouteDeclaration(None=DotProduct(A=NamedRerouteUsage(), B=Constant() {r=0.25})) {name=Combined, nodeColor=<rgba 0 1 0.0859 1>}
  NamedRerouteUsage()
 ]
}
